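# Anim Blueprint for the BlobRock pickaxe that adds secondary physics motion to its front and back dumbbell bones.
# The AnimGraph passes the input pose through two chained, root-relative AnimDynamics nodes with angular springs, tight angle limits and gravity override.
# Each update sets gravity from a deprecated helper, then FE Alpha and simulation space by front-end and on-pawn state.
# The AnimGraph shows no reads of these variables, and both nodes run at full alpha.
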AnimBlueprint /BRCosmetics/Weapons/FORT_Melee/Pickaxe_BlobRock/Meshes/BlobRock_Axe_AnimBP
target_skeleton=&/BlobRock_Axe_Skeleton
graphs [
 AnimGraph {nodes [
   AnimGraphNode_SubInput_0 = Input Pose()
   n1 = Anim Dynamics - Bone: dyn_dumbbell_fr!(
    SimulationSpace:AnimPhysSimSpaceType Enum
    ComponentPose=Local To Component(LocalPose=@AnimGraphNode_SubInput_0)
    Alpha=1
    ExternalForce="0,0,0"
   ) -> (Pose) {
    angularDampingOverride=1
    angularSpringConstant=10000
    bAngularSpring=true
    bOverrideAngularDamping=true
    bUseGravityOverride=true
    simulationSpace=RootRelative
    alphaBoolBlend {blendOption=Linear}
    boundBone {boneName=dyn_dumbbell_fr}
    componentAppliedLinearAccClamp=<2000 2000 2000>
    componentLinearAccScale=<0.4 0.4 0.4>
    componentLinearVelScale=<0.4 0.4 0.4>
    gravityOverride {z=-980}
    physicsBodyDefinitions [{
      boundBone {boneName=dyn_dumbbell_fr}
      boxExtents=<10 10 10>
      constraintSetup {angularLimitsMax=<2 10 20>, angularLimitsMin=<-2 -10 -2>, angularTarget=<1 0 0>}
      localJointOffset=<2 0 0>
     }]
   }
   n2 = Anim Dynamics - Bone: dyn_dumbbell_bk!(SimulationSpace:AnimPhysSimSpaceType Enum, ComponentPose=@n1, Alpha=1, ExternalForce="0,0,0") -> (Pose) {
    angularDampingOverride=1
    angularSpringConstant=10000
    bAngularSpring=true
    bOverrideAngularDamping=true
    bUseGravityOverride=true
    simulationSpace=RootRelative
    alphaBoolBlend {blendOption=Linear}
    boundBone {boneName=dyn_dumbbell_bk}
    componentAppliedLinearAccClamp=<2000 2000 2000>
    componentLinearAccScale=<0.4 0.4 0.4>
    componentLinearVelScale=<0.4 0.4 0.4>
    gravityOverride {z=-980}
    physicsBodyDefinitions [{
      boundBone {boneName=dyn_dumbbell_bk}
      boxExtents=<10 10 10>
      constraintSetup {angularLimitsMax=<2 10 20>, angularLimitsMin=<-2 -10 -2>, angularTarget=<1 0 0>}
      localJointOffset=<2 0 0>
     }]
   }
   Output Pose(Result=Component To Local(ComponentPose=@n2)) {layerGroup=None}
  ]}
 EventGraph {nodes [
   Event BlueprintUpdateAnimation()
   > n1 = Gravity Override State Function DEPRECATED(CharacterPartAnimInstance=Self-Reference(), GameTimeInSeconds=GetGameTimeInSeconds(), GravityOverrideParams=Get Gravity Override Params())
   > Set Gravity Override(Gravity Override=@n1)
   > n2 = Branch(Condition=Get bIsInFrontEnd())
   > n3 = Branch(Condition=Get bIsOnPawn() -> (bIsOnPawn:bool))
   > Set FE Alpha(FE Alpha:double=0.7) -> (Output_Get:double)
   > Set Simulation Space(Simulation Space:AnimPhysSimSpaceType Enum) -> (Output_Get:AnimPhysSimSpaceType Enum)
   Set FE Alpha(execute=@n3.else, FE Alpha=0.2)
   > Set Simulation Space(Simulation Space=World)
   Set FE Alpha(execute=@n2.else, FE Alpha=1)
   > Set Simulation Space()
   TryGetPawnOwner()
  ]}
]
variables [
 FE Alpha {type=number}
 Gravity Override {type=Vector}
 Gravity Override Params {type=GravityOverrideParamsStruct}
 Simulation Space {type=AnimPhysSimSpaceType}
]
